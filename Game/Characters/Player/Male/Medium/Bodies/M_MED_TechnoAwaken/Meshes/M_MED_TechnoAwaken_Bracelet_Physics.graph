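# Physics asset for the bracelet mesh of the medium male TechnoAwaken player body, defining how the bracelet moves on the left forearm.
# Two sphere bodies cover the dyn_bracelet bone and the lowerarm_twist_01_l bone.
# A single constraint joins them with tight swing and twist limits of a few degrees, allowing only slight rotational play of the bracelet relative to the arm.
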PhysicsAsset /Game/Characters/Player/Male/Medium/Bodies/M_MED_TechnoAwaken/Meshes/M_MED_TechnoAwaken_Bracelet_Physics
bodies {
 dyn_bracelet [{shapeType=Sphere, radius=2.7301, shapeName=dyn_bracelet_sphere, center=<0.0775 0.2295 0.2537>}]
 lowerarm_twist_01_l [{shapeType=Sphere, radius=2.0475, shapeName=lowerarm_twist_01_l_sphere, center=<-0.5862 0.0685 0.6632>}]
}
constraints [{bone1Name=dyn_bracelet, bone2Name=lowerarm_twist_01_l, swing1LimitDegrees=4, swing2LimitDegrees=3, twistLimitDegrees=3}]
