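# Physics asset for the chain and medallion on the male medium TechnoAwaken player body.
# Capsules on spine_02 and spine_05 form the torso, while capsules on dyn_chain_root and dyn_chain and a box on dyn_medallion form the pendant.
# Constraints link spine_05 through root and chain to the medallion, plus the two spine bodies.
# Selectively locked axes keep the root nearly rigid, restrict chain swing and stop medallion twist.
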PhysicsAsset /Game/Characters/Player/Male/Medium/Bodies/M_MED_TechnoAwaken/Meshes/M_MED_TechnoAwaken_Chain_Physics
bodies {
 dyn_chain [{shapeType=Capsule, length=3.7695, radius=0.9244, shapeName=dyn_chain_capsule, center=<0 1.6721 -0.4766>, rotation=<rot 0 0 -80.0004>}]
 dyn_chain_root [{
   shapeType=Capsule
   length=2.7253
   radius=1.0074
   shapeName=dyn_chain_root_capsule
   center=<-0.0428 1.1908 -0.0678>
   rotation=<rot 0.0761 0.345 -90.0003>
  }]
 dyn_medallion [{
   shapeType=Box
   extentX=0.9346
   extentY=6.7933
   extentZ=8.1562
   shapeName=dyn_medallion_box
   center=<-3.1852 -0.3656 -0.0023>
   rotation=<rot 0 96.2283 -89.9973>
  }]
 spine_02 [{shapeType=Capsule, length=13.9865, radius=10.3143, center=<9.1603 1.4887 0.0472>, rotation=<rot -0.7448 -90 0.0335>}]
 spine_05 [{shapeType=Capsule, length=15.9495, radius=10.4305, center=<-2.6437 0.8936 -0.0002>, rotation=<rot 0.0124 -90 -0.0086>}]
}
constraints [
 {bone1Name=spine_05, bone2Name=spine_02, swing1LimitDegrees=45, swing2LimitDegrees=45, twistLimitDegrees=45}
 {bone1Name=dyn_chain, bone2Name=dyn_chain_root, swing1LimitDegrees=2, swing2LimitDegrees=45, swing2Motion=Locked, twistLimitDegrees=20}
 {bone1Name=dyn_medallion, bone2Name=dyn_chain, swing1LimitDegrees=40, swing2LimitDegrees=10, twistLimitDegrees=45, twistMotion=Locked}
 {
  bone1Name=dyn_chain_root
  bone2Name=spine_05
  swing1LimitDegrees=2
  swing1Motion=Locked
  swing2LimitDegrees=45
  swing2Motion=Locked
  twistLimitDegrees=5
 }
]
