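# Physics asset for the left leg mesh of the TechnoAwaken medium male player body.
# Capsules cover the calf and foot, while five dyn_calf_twist bones add small capsules around the calf plus an upper sphere.
# All five constrain to calf_l with swing and twist locked; the foot-to-calf joint is not angularly locked.
# The dyn shapes carry right-side names, suggesting they are mirrored from the right leg.
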PhysicsAsset /Game/Characters/Player/Male/Medium/Bodies/M_MED_TechnoAwaken/Meshes/M_MED_TechnoAwaken_Leg_L_Physics
bodies {
 calf_l [{shapeType=Capsule, length=42.0486, radius=8.285, center=<-24.0001 -1.2457 0.8816>, rotation=<rot -0.0071 89.3909 -90.1098>}]
 dyn_calf_twist_back_l [{
   shapeType=Capsule
   length=4.63
   radius=2.63
   shapeName=dyn_calf_twist_back_r_capsule
   center=<0.2743 -4.7734 -3.172>
   rotation=<rot 2.0752 -90 0>
  }]
 dyn_calf_twist_front_l [{
   shapeType=Capsule
   length=1.75
   radius=2.75
   shapeName=dyn_calf_twist_front_r_capsule
   center=<0.6769 -0.132 -1.7911>
   rotation=<rot 3.9631 -99.7004 -4.1459>
  }]
 dyn_calf_twist_left_l [{
   shapeType=Capsule
   length=6.255
   radius=2.63
   shapeName=dyn_calf_twist_right_r_capsule
   center=<3.911 -2.0565 -1.9479>
   rotation=<rot -2.5094 -88.4775 -6.4658>
  }]
 dyn_calf_twist_right_l [{
   shapeType=Capsule
   length=5.5
   radius=2.38
   shapeName=dyn_calf_twist_left_r_capsule
   center=<-2.0859 0.0001 -3.0009>
   rotation=<rot -0.0853 -89.9967 -7.8103>
  }]
 dyn_calf_twist_upper_l [{shapeType=Sphere, radius=3.5793, shapeName=dyn_calf_twist_upper_r_sphere, center=<0.167 0.1883 -0.12>}]
 foot_l [{shapeType=Capsule, length=17.4494, radius=7.9828, center=<-3.7028 6.5228 0.0583>, rotation=<rot 0 0.8667 -89.4543>}]
}
constraints [
 {bone1Name=foot_l, bone2Name=calf_l, swing1LimitDegrees=45, swing2LimitDegrees=45, twistLimitDegrees=45}
 {
  bone1Name=dyn_calf_twist_back_l
  bone2Name=calf_l
  swing1LimitDegrees=45
  swing1Motion=Locked
  swing2LimitDegrees=5
  swing2Motion=Locked
  twistLimitDegrees=10
  twistMotion=Locked
 }
 {
  bone1Name=dyn_calf_twist_front_l
  bone2Name=calf_l
  swing1LimitDegrees=45
  swing1Motion=Locked
  swing2LimitDegrees=5
  swing2Motion=Locked
  twistLimitDegrees=10
  twistMotion=Locked
 }
 {
  bone1Name=dyn_calf_twist_left_l
  bone2Name=calf_l
  swing1LimitDegrees=45
  swing1Motion=Locked
  swing2LimitDegrees=10
  swing2Motion=Locked
  twistLimitDegrees=10
  twistMotion=Locked
 }
 {
  bone1Name=dyn_calf_twist_right_l
  bone2Name=calf_l
  swing1LimitDegrees=45
  swing1Motion=Locked
  swing2LimitDegrees=10
  swing2Motion=Locked
  twistLimitDegrees=10
  twistMotion=Locked
 }
 {
  bone1Name=dyn_calf_twist_upper_l
  bone2Name=calf_l
  swing1LimitDegrees=45
  swing1Motion=Locked
  swing2LimitDegrees=45
  swing2Motion=Locked
  twistLimitDegrees=20
  twistMotion=Locked
 }
]
