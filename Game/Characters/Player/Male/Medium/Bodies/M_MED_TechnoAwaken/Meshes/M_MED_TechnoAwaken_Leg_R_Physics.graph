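# Physics asset for the right leg mesh of the medium male TechnoAwaken player body.
# Capsules cover the calf and foot, while a sphere and four capsules cover the upper, front, back, left and right dyn_calf_twist bones.
# All six constraints attach directly to calf_r in a star layout.
# The five dyn bones have swing and twist locked, so their stored angle limits are inactive; the foot's rotation is not locked.
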PhysicsAsset /Game/Characters/Player/Male/Medium/Bodies/M_MED_TechnoAwaken/Meshes/M_MED_TechnoAwaken_Leg_R_Physics
bodies {
 calf_r [{shapeType=Capsule, length=40.936, radius=8.9353, center=<23.2578 1.0568 -0.6843>, rotation=<rot 0.0211 88.5495 -89.6133>}]
 dyn_calf_twist_back_r [{
   shapeType=Capsule
   length=4.63
   radius=2.755
   shapeName=dyn_calf_twist_back_r_capsule
   center=<0.2743 -4.6729 -4.4211>
   rotation=<rot 2.6942 -90 0>
  }]
 dyn_calf_twist_front_r [{
   shapeType=Capsule
   length=2.75
   radius=2.75
   shapeName=dyn_calf_twist_front_r_capsule
   center=<-0.3762 1.4797 -2.0202>
   rotation=<rot 9.2645 -90.7285 3.493>
  }]
 dyn_calf_twist_left_r [{
   shapeType=Capsule
   length=5.5
   radius=2.38
   shapeName=dyn_calf_twist_left_r_capsule
   center=<3.4378 0 -2.7159>
   rotation=<rot -0.0849 -89.9968 2.45>
  }]
 dyn_calf_twist_right_r [{
   shapeType=Capsule
   length=6.255
   radius=2.63
   shapeName=dyn_calf_twist_right_r_capsule
   center=<-2.9315 -0.7916 -3.0343>
   rotation=<rot 10.8054 -90 0>
  }]
 dyn_calf_twist_upper_r [{shapeType=Sphere, radius=3.5793, shapeName=dyn_calf_twist_upper_r_sphere, center=<0.167 0.1883 -0.12>}]
 foot_r [{shapeType=Capsule, length=17.4494, radius=7.9828, center=<2.8738 -6.5742 -0.1097>, rotation=<rot 0 0.8667 -89.4543>}]
}
constraints [
 {bone1Name=foot_r, bone2Name=calf_r, swing1LimitDegrees=45, swing2LimitDegrees=45, twistLimitDegrees=45}
 {
  bone1Name=dyn_calf_twist_upper_r
  bone2Name=calf_r
  swing1LimitDegrees=45
  swing1Motion=Locked
  swing2LimitDegrees=45
  swing2Motion=Locked
  twistLimitDegrees=20
  twistMotion=Locked
 }
 {
  bone1Name=dyn_calf_twist_back_r
  bone2Name=calf_r
  swing1LimitDegrees=45
  swing1Motion=Locked
  swing2LimitDegrees=5
  swing2Motion=Locked
  twistLimitDegrees=10
  twistMotion=Locked
 }
 {
  bone1Name=dyn_calf_twist_front_r
  bone2Name=calf_r
  swing1LimitDegrees=45
  swing1Motion=Locked
  swing2LimitDegrees=5
  swing2Motion=Locked
  twistLimitDegrees=10
  twistMotion=Locked
 }
 {
  bone1Name=dyn_calf_twist_left_r
  bone2Name=calf_r
  swing1LimitDegrees=45
  swing1Motion=Locked
  swing2LimitDegrees=10
  swing2Motion=Locked
  twistLimitDegrees=10
  twistMotion=Locked
 }
 {
  bone1Name=dyn_calf_twist_right_r
  bone2Name=calf_r
  swing1LimitDegrees=45
  swing1Motion=Locked
  swing2LimitDegrees=10
  swing2Motion=Locked
  twistLimitDegrees=10
  twistMotion=Locked
 }
]
