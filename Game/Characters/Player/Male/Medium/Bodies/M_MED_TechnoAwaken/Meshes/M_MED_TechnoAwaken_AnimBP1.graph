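# Post-process style Anim Blueprint for the TechnoAwaken player body, adding secondary physics motion and material-driven cloth deformation.
# Copies the attached parent's pose, then chains RigidBody sims for chain, bracelet and both legs into a Control Rig; an AnimDynamics bracelet node is unconnected.
# Each update feeds foot velocity and calf collision into WPO functions on body and overlay dynamic materials.
# Cloth wind switches on during skydiving or parachuting.
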AnimBlueprint /Game/Characters/Player/Male/Medium/Bodies/M_MED_TechnoAwaken/Meshes/M_MED_TechnoAwaken_AnimBP1
target_skeleton=&/M_MED_TechnoAwaken_Skeleton
graphs [
 AnimGraph {nodes [
   Anim Dynamics - Bone: dyn_bracelet(Alpha=1, ExternalForce="0,0,0") {
    angularDampingOverride=1
    angularSpringConstant=0.1
    bAngularSpring=true
    bLinearSpring=true
    bOverrideLinearDamping=true
    bUseGravityOverride=true
    linearDampingOverride=1
    linearSpringConstant=0.1
    simulationSpace=RootRelative
    alphaBoolBlend {blendOption=Linear}
    boundBone {boneName=dyn_bracelet}
    gravityOverride {z=-980}
    physicsBodyDefinitions [{
      boundBone {boneName=dyn_bracelet}
      boxExtents=<10 10 10>
      constraintSetup {angularLimitsMax=<3 3 3>, angularLimitsMin=<-3 -3 -3>, angularTarget=<0 0 0>, linearAxesMax=<0.1 0.1 0.15>, linearAxesMin=<-0.8 -0.1 -0.15>}
      localJointOffset=<0 0 0>
     }]
    relativeSpaceBone {boneName=lowerarm_l}
   }
   n1 = RigidBody(ComponentPose=Local To Component(LocalPose=Copy Pose From Mesh() {bUseAttachedParent=true})) {
    bOverrideWorldGravity=true
    overridePhysicsAsset=&/M_MED_TechnoAwaken_Chain_Physics
    simulationSpace=BaseBoneSpace
    alphaBoolBlend {blendOption=Linear}
    baseBoneRef {boneName=spine_05}
    componentAppliedLinearAccClamp=<5000 5000 5000>
    overrideWorldGravity {z=-980}
    simSpaceSettings {maxAngularAcceleration=500, maxAngularVelocity=10, maxLinearAcceleration=350, maxLinearVelocity=100, velocityScaleZ=0, worldAlpha=0.35}
   }
   n2 = RigidBody(ComponentPose=@n1) {
    bOverrideWorldGravity=true
    overridePhysicsAsset=&/M_MED_TechnoAwaken_Bracelet_Physics
    simulationSpace=BaseBoneSpace
    alphaBoolBlend {blendOption=Linear}
    baseBoneRef {boneName=lowerarm_twist_01_l}
    componentAppliedLinearAccClamp=<5000 5000 5000>
    overrideWorldGravity {z=-980}
    simSpaceSettings {maxAngularAcceleration=500, maxAngularVelocity=10, maxLinearAcceleration=350, maxLinearVelocity=100, velocityScaleZ=0, worldAlpha=0.45}
   }
   n3 = RigidBody(ComponentPose=@n2) {
    bOverrideWorldGravity=true
    overridePhysicsAsset=&/M_MED_TechnoAwaken_Leg_R_Physics
    simulationSpace=BaseBoneSpace
    alphaBoolBlend {blendOption=Linear}
    alphaScaleBiasClamp {bMapRange=true, outRange {max=0, min=1}}
    baseBoneRef {boneName=calf_r}
    componentAppliedLinearAccClamp=<8000 8000 8000>
    componentLinearAccScale=<0.6 0.6 0.6>
    componentLinearVelScale=<0.1 0.1 0.1>
    overrideWorldGravity {z=-980}
    simSpaceSettings {maxAngularAcceleration=500, maxAngularVelocity=200, maxLinearAcceleration=400, maxLinearVelocity=200, velocityScaleZ=0, worldAlpha=0.85}
   }
   n4 = RigidBody(ComponentPose=@n3) {
    bOverrideWorldGravity=true
    overridePhysicsAsset=&/M_MED_TechnoAwaken_Leg_L_Physics
    simulationSpace=BaseBoneSpace
    alphaBoolBlend {blendOption=Linear}
    alphaScaleBiasClamp {bMapRange=true, outRange {max=0, min=1}}
    baseBoneRef {boneName=calf_l}
    componentAppliedLinearAccClamp=<8000 8000 8000>
    componentLinearAccScale=<0.6 0.6 0.6>
    componentLinearVelScale=<0.1 0.1 0.1>
    overrideWorldGravity {z=-980}
    simSpaceSettings {maxAngularAcceleration=500, maxAngularVelocity=200, maxLinearAcceleration=400, maxLinearVelocity=200, velocityScaleZ=0, worldAlpha=0.85}
   }
   n5 = Control Rig(Source=Component To Local(ComponentPose=@n4)) {
    alphaBoolBlend {blendOption=Linear}
    controlRigAssetReference {blueprintRigClass=&/M_MED_TechnoAwaken_CtrlRig.M_MED_TechnoAwaken_CtrlRig_C}
   }
   Output Pose(Result=@n5) {layerGroup=None}
  ]}
 EventGraph {nodes [
   n1 = Event BlueprintUpdateAnimation()
   > Sequence()
   > Is Valid(InputObject=Get Player Pawn() -> (Player Pawn:Fort Player Pawn Athena Object Reference))
   > Set WPO Globals Function(
    Material=Get Material Instance Dynamic() -> (Material Instance Dynamic:Material Instance Dynamic Object Reference)
    Overlay Material=Get Overlay Material Instance Dynamic() -> (Overlay Material Instance Dynamic:Material Instance Dynamic Object Reference)
    Skeletal Mesh=Get Skeletal Mesh() -> (Skeletal Mesh:Skeletal Mesh Component Object Reference)
    "Local Noise Transform  Location"="Get Local Noise Transform  Location"() -> ("Local Noise Transform  Location":Vector)
    Local Noise Offset=Get Local Noise Offset()
   )
   > Set WPO Regional Function!(
    execute
    Material:Material Instance Dynamic Object Reference=Get Material Instance Dynamic()
    Overlay Material:Material Instance Dynamic Object Reference=Get Overlay Material Instance Dynamic()
    Skeletal Mesh:Skeletal Mesh Component Object Reference=Get Skeletal Mesh()
    Velocity Transform:Name=foot_r
    Velocity Interp Speed:double=Get Velocity Interp Speed()
    Collision Transform:Name=calf_r
    Collision Transform Axis:EAxis Enum
    Invert Collision Transform Axis:bool=false
    Delta Time X:double=@n1.DeltaTimeX
    Previous Velocity Transform Location:"Vector (by ref)"=Get R1 Velocity Transform Location()
    Previous Velocity Transform Velocity:"Vector (by ref)"=Get R1 Velocity Transform Velocity()
    Region:int=1
   ) -> (then)
   > Set WPO Regional Function(
    Material=Get Material Instance Dynamic()
    Overlay Material=Get Overlay Material Instance Dynamic()
    Skeletal Mesh=Get Skeletal Mesh()
    Velocity Transform=foot_l
    Velocity Interp Speed=Get Velocity Interp Speed()
    Collision Transform=calf_l
    Invert Collision Transform Axis=true
    Delta Time X=@n1.DeltaTimeX
    Previous Velocity Transform Location=Get R2 Velocity Transform Location()
    Previous Velocity Transform Velocity=Get R2 Velocity Transform Velocity()
    Region=2
   )
   > n2 = Set isWPOEnabled(isWPOEnabled:bool=OR Boolean!(A=Get bIsSkydiveFloating(), B=Get bIsSkydiveDiveMode(), C:bool=Get bIsParachuteOpened()) -> (ReturnValue)) -> (Output_Get:bool)
   > n3 = Branch(Condition=@n2)
   > n4 = Set WPO Pho Wind Intensity(WPO Pho Wind Intensity=1)
   SetScalarParameterValue(
    execute=[@n5.then, @n4.then]
    self=[Get Overlay Material Instance Dynamic(), Get Material Instance Dynamic()]
    ParameterName=ClothSim_WindStrength
    Value=Get WPO Pho Wind Intensity()
   )
   n5 = Set WPO Pho Wind Intensity(execute=@n3.else, WPO Pho Wind Intensity=0)
   Event BlueprintInitializeAnimation()
   > Set Skeletal Mesh(Skeletal Mesh:Skeletal Mesh Component Object Reference=GetOwnerSkeletalMeshComponent()) -> (Output_Get:Skeletal Mesh Component Object Reference)
   n6 = TryGetPawnOwner()
   > Is Valid(InputObject=@n6)
   > n7 = Cast To FortPlayerPawnAthena(Object=@n6)
   > Set Player Pawn(Player Pawn:Fort Player Pawn Athena Object Reference=@n7) -> (Output_Get:Fort Player Pawn Athena Object Reference)
   Event OnCharacterCustomizationLoaded()
   > n8 = Set Skeletal Mesh(Skeletal Mesh=GetPartSkeletalMeshComponent())
   > n9 = Get Dynamic Material Instance(SkeletalMeshComponent=@n8)
   > Set Material Instance Dynamic(Material Instance Dynamic:Material Instance Dynamic Object Reference=@n9) -> (Output_Get:Material Instance Dynamic Object Reference)
   > n10 = GetOrCreateDynamicOverlayMaterial(MeshComponent=@n8, MaterialName=MI_TechnoAwaken_Body)
   > Set Overlay Material Instance Dynamic(Overlay Material Instance Dynamic:Material Instance Dynamic Object Reference=@n10) -> (Output_Get:Material Instance Dynamic Object Reference)
  ]}
]
variables [
 ClothSim_WindStrength {type=string, default=ClothSim_WindStrength}
 isWPOEnabled {type=boolean}
 Local Noise Offset {type=Vector}
 "Local Noise Transform  Location" {type=Vector}
 Material Instance Dynamic {type=MaterialInstanceDynamic}
 Overlay Material Instance Dynamic {type=MaterialInstanceDynamic}
 Player Pawn {type=FortPlayerPawnAthena}
 R1 Velocity Transform Location {type=Vector}
 R1 Velocity Transform Velocity {type=Vector}
 R2 Velocity Transform Location {type=Vector}
 R2 Velocity Transform Velocity {type=Vector}
 Skeletal Mesh {type=SkeletalMeshComponent}
 Velocity Interp Speed {type=number}
 WPO Pho Wind Intensity {type=number, default=1}
]
